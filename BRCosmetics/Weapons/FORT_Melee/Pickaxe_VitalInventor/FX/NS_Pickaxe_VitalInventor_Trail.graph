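# Niagara ribbon trail effect for the VitalInventor pickaxe, built from three anim-trail emitters that follow bones or sockets on the skeletal mesh.
# Main_Trail draws a warm orange ribbon that narrows and fades over life, with RGB scaled by time of day.
# Left and right eye trails are faint, short-lived pale streaks on single bone indices, offset vertically by spawn group.
# Each eye trail uses two ribbon renderers and sets a dark inner color and width.
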NiagaraSystem /BRCosmetics/Weapons/FORT_Melee/Pickaxe_VitalInventor/FX/NS_Pickaxe_VitalInventor_Trail
dependencies {
 usedDataInterfaces [&NiagaraDataInterfaceArrayFloat3, &NiagaraDataInterfaceCurve, &NiagaraDataInterfaceSkeletalMesh]
 usedDynamicInputs []
 usedModules [
  &SystemState
  &self:SystemSpawnScript.SystemScriptSource.SystemScriptGraph.NiagaraNodeAssignment_0.SetVariables_#1
 ]
}
emitters {
 Left_Eye_Trail {
  emitterUpdateScript [
   &EmitterState()
   &FN_SpawnAnimTrail()
  ]
  particleSpawnScript [
   &FN_InitializeAnimTrailBoneIndecies(First Bone or Socket Index=2, Second Bone or Socket Index=2)
   &FN_InitializeAnimTrail(
    AnimTrailWidth=10
    Color=&FNTimeOfDayColor(
     Night Color=<rgba 0.7604 0.8181 0.8438 0.01>
     Morning Color=<rgba 0.7604 0.8181 0.8438 0.02>
     Day Color=<rgba 0.7604 0.8181 0.8438 0.025>
     Evening Color=<rgba 0.7604 0.8181 0.8438 0.02>
    )
    Lifetime=0.05
    Position Offset=&SelectVectorFromArray(Direct Array Index=$Particles.SpawnGroup, Vector Selection Array=DataInterface() {propertyValues {FloatData [<0 0 -1>, <0 0 1>]}})
    SkeletalMesh=$System.Skeletal Mesh
   )
   /SetVariables(Particles.InnerColor:LinearColor=<rgba 0 0 0 1.25>, Particles.InnerWidth=2.5) {isSetParameters=true, module=SetVariables_#2}
  ]
  particleUpdateScript [
   &SolveForcesAndVelocity(Clamp Velocity=false, Limit Acceleration=false)
   &FN_UpdateAnimTrail(Locked Particle RibbonWidth=1, SkeletalMesh=$System.Skeletal Mesh) {module=FN_UpdateAnimTrail001}
   &ScaleColor(
    Scale Alpha=&FloatFromCurve(
     FloatCurve=DataInterface() {propertyValues {Curve {keys [{interpMode=RCIM_Cubic, tangentMode=RCTM_User, time=0.5, value=1}, {interpMode=RCIM_Cubic, tangentMode=RCTM_User, time=1, value=0}]}}}
     Scale Curve=7
    )
    ScaleA=true
    ScaleRGB=false
   )
  ]
  rendererClasses [&NiagaraRibbonRendererProperties]
  renderers [
   {rendererClass=&NiagaraRibbonRendererProperties}
   {rendererClass=&NiagaraRibbonRendererProperties, rendererIndex=1}
  ]
 }
 Main_Trail {
  emitterUpdateScript [
   &EmitterState(Life Cycle Mode=::NewEnumerator1, Loop Behavior=::NewEnumerator1, Loop Duration=0.27, UseLoopDelay=false)
   &FN_SpawnAnimTrail()
  ]
  particleSpawnScript [
   &FN_InitializeAnimTrailBoneIndecies()
   &FN_InitializeAnimTrail(
    Color=&FNTimeOfDayColor(
     Night Color=<rgba 0.651 0.3006 0.0533 0.75>
     Morning Color=<rgba 0.75 0.4237 0.075 0.75>
     Day Color=<rgba 0.75 0.4237 0.075 0.75>
     Evening Color=<rgba 0.651 0.3006 0.0533 0.75>
    )
    Lifetime=0.14
    SkeletalMesh=$System.Skeletal Mesh
   )
  ]
  particleUpdateScript [
   &ParticleState()
   &ScaleRibbonWidth(Ribbon Width Scale=&FloatFromCurve(FloatCurve=DataInterface() {propertyValues {Curve {keys [
         {arriveTangent=-1.2579, interpMode=RCIM_Cubic, time=0, value=1}
         {arriveTangent=-3.0042, interpMode=RCIM_Cubic, leaveTangent=-3.0042, tangentMode=RCTM_User, time=1, value=0}
        ]}}}))
   &ScaleColor(
    Scale Alpha=&Multiply_Float(
     A=&FloatFromCurve(
      FloatCurve=DataInterface() {propertyValues {Curve {keys [
          {interpMode=RCIM_Cubic, leaveTangent=-2.3646, tangentMode=RCTM_Break, time=0, value=1}
          <at 0.6689 0.0233>
          {arriveTangent=-0.2139, interpMode=RCIM_Cubic, tangentMode=RCTM_Break, time=1, value=0}
         ]}}}
      Scale Curve=0.75
     )
     B=1
    )
    Scale RGB=&Multiply_VectorByFloat(
     Vector=&MakeVectorFromLinearColorRGB(Color=$Particles.Initial.Color)
     Float=&FNTimeOfDayScalar(Night Scale=0.5, Morning Scale=0.75, Evening Scale=0.5)
    )
    ScaleA=true
    ScaleRGB=true
   )
   &FN_UpdateAnimTrail(SkeletalMesh=$System.Skeletal Mesh)
   /SetVariables(Particles.RibbonV0RangeOverride:Vector2f=<0.125 0>) {isSetParameters=true}
  ]
  rendererClasses [&NiagaraRibbonRendererProperties]
  renderers [{rendererClass=&NiagaraRibbonRendererProperties}]
 }
 Right_Eye_Trail {
  emitterUpdateScript [
   &EmitterState()
   &FN_SpawnAnimTrail()
  ]
  particleSpawnScript [
   &FN_InitializeAnimTrailBoneIndecies(First Bone or Socket Index=3, Second Bone or Socket Index=3)
   &FN_InitializeAnimTrail(
    AnimTrailWidth=10
    Color=&FNTimeOfDayColor(
     Night Color=<rgba 0.7604 0.8181 0.8438 0.01>
     Morning Color=<rgba 0.7604 0.8181 0.8438 0.02>
     Day Color=<rgba 0.7604 0.8181 0.8438 0.025>
     Evening Color=<rgba 0.7604 0.8181 0.8438 0.02>
    )
    Lifetime=0.05
    Position Offset=&SelectVectorFromArray(Direct Array Index=$Particles.SpawnGroup, Vector Selection Array=DataInterface() {propertyValues {FloatData [<0 0 -1>, <0 0 1>]}})
    SkeletalMesh=$System.Skeletal Mesh
   )
   /SetVariables(Particles.InnerColor:LinearColor=<rgba 0 0 0 1.25>, Particles.InnerWidth=2.5) {isSetParameters=true, module=SetVariables_#2}
  ]
  particleUpdateScript [
   &SolveForcesAndVelocity(Clamp Velocity=false, Limit Acceleration=false)
   &FN_UpdateAnimTrail(Locked Particle RibbonWidth=1, SkeletalMesh=$System.Skeletal Mesh) {module=FN_UpdateAnimTrail001}
   &ScaleColor(
    Scale Alpha=&FloatFromCurve(
     FloatCurve=DataInterface() {propertyValues {Curve {keys [{interpMode=RCIM_Cubic, tangentMode=RCTM_User, time=0.5, value=1}, {interpMode=RCIM_Cubic, tangentMode=RCTM_User, time=1, value=0}]}}}
     Scale Curve=7
    )
    ScaleA=true
    ScaleRGB=true
   )
  ]
  rendererClasses [&NiagaraRibbonRendererProperties]
  renderers [
   {rendererClass=&NiagaraRibbonRendererProperties}
   {rendererClass=&NiagaraRibbonRendererProperties, rendererIndex=1}
  ]
 }
}
summary {
 systemName=NS_Pickaxe_VitalInventor_Trail
 emitters [
  {emitterName=Left_Eye_Trail, rendererClasses [&NiagaraRibbonRendererProperties]}
  {emitterName=Right_Eye_Trail, rendererClasses [&NiagaraRibbonRendererProperties]}
  {emitterName=Main_Trail, rendererClasses [&NiagaraRibbonRendererProperties]}
 ]
 userVariables [
  User.bIsFrontEndPreview {type=NiagaraBool}
  User.GradientTexture {
   type=Texture
   description=""
   defaultValue {
    struct=NiagaraExt_VariableValue_Object
    value {object=/BRCosmetics/Weapons/FORT_Melee/Pickaxe_VitalInventorBlock/FX/T_VitalInventor_Gradient, objectClass=/Script/Engine.Texture}
   }
  }
  User.LightColor {type=LinearColor, description="", defaultValue {struct=LinearColor, value=<rgba 1 1 1 1>}}
  User.NewVector3f {type=Vector3f, description="", defaultValue {struct=Vector3f, value=<0 0 0>}}
  User.RibbonIDOffset {type=NiagaraInt32}
  User.VariantPicker {type=NiagaraFloat, defaultValue=0, description=""}
 ]
}
