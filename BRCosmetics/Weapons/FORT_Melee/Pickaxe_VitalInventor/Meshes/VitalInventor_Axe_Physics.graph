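# Physics asset for the VitalInventor pickaxe, a two-handed melee cosmetic with a head and two eye chains on dynamic bones.
# Capsule bodies cover the handle, the head and three bones in each eye chain.
# The head joint to the handle allows only limited twist, with zero-degree swing limits.
# Each eye chain links from its base on the head out to the eye through joints with small, uniform swing and twist limits.
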PhysicsAsset /BRCosmetics/Weapons/FORT_Melee/Pickaxe_VitalInventor/Meshes/VitalInventor_Axe_Physics
bodies {
 Melee_TwoHanded_Handle [{shapeType=Capsule, length=58.847, radius=19.5786, center=<0.8793 1.526 20.8987>, rotation=<rot 4.802 -90 -0.0015>}]
 dyn_eye_l [{
   shapeType=Capsule
   length=5.9749
   radius=4.7157
   shapeName=dyn_eye_l_capsule
   center=<3.6339 0.5265 0.1904>
   rotation=<rot 0 120.0632 -56.823>
  }]
 dyn_eye_l_01 [{
   shapeType=Capsule
   length=3.757
   radius=3.483
   shapeName=dyn_eye_l_01_capsule
   center=<-0.2091 -0.7344 -2.3581>
   rotation=<rot -28.8408 -90 25.4716>
  }]
 dyn_eye_l_base [{
   shapeType=Capsule
   length=4.6631
   radius=4.5467
   shapeName=dyn_eye_l_base_capsule
   center=<4.2072 -0.6227 0.3867>
   rotation=<rot 0 76.3369 -95.1033>
  }]
 dyn_eye_r [{
   shapeType=Capsule
   length=5.9749
   radius=4.7157
   shapeName=dyn_eye_l_capsule
   center=<-2.0238 1.4874 1.3657>
   rotation=<rot -21.7709 101.845 41.5829>
  }]
 dyn_eye_r_01 [{
   shapeType=Capsule
   length=4.6631
   radius=4.5467
   shapeName=dyn_eye_l_base_capsule
   center=<0.9514 -0.7494 -0.2055>
   rotation=<rot 0 76.3369 -95.1033>
  }]
 dyn_eye_r_base [{
   shapeType=Capsule
   length=4.6631
   radius=3.5602
   shapeName=dyn_eye_r_base_capsule
   center=<-2.3531 -0.8644 -1.7986>
   rotation=<rot 0 -79.9064 -52.3682>
  }]
 dyn_head [{shapeType=Capsule, length=16.4925, radius=9.8001, center=<0.1118 4.7228 6.8484>, rotation=<rot 1.5226 -176.9847 -103.0747>}]
}
constraints [
 {bone1Name=dyn_head, bone2Name=Melee_TwoHanded_Handle, swing1LimitDegrees=0, swing2LimitDegrees=0, twistLimitDegrees=30}
 {bone1Name=dyn_eye_l, bone2Name=dyn_eye_l_01, swing1LimitDegrees=7, swing2LimitDegrees=7, twistLimitDegrees=7}
 {bone1Name=dyn_eye_l_base, bone2Name=dyn_head, swing1LimitDegrees=7, swing2LimitDegrees=7, twistLimitDegrees=7}
 {bone1Name=dyn_eye_r_base, bone2Name=dyn_head, swing1LimitDegrees=7, swing2LimitDegrees=7, twistLimitDegrees=7}
 {bone1Name=dyn_eye_r, bone2Name=dyn_eye_r_01, swing1LimitDegrees=7, swing2LimitDegrees=7, twistLimitDegrees=7}
 {bone1Name=dyn_eye_r_01, bone2Name=dyn_eye_r_base, swing1LimitDegrees=7, swing2LimitDegrees=7, twistLimitDegrees=7}
 {bone1Name=dyn_eye_l_01, bone2Name=dyn_eye_l_base, swing1LimitDegrees=7, swing2LimitDegrees=7, twistLimitDegrees=7}
]
