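# Anim Blueprint for the VitalInventor pickaxe that adds physics-driven secondary motion to its dangling eye bones.
# The cached input pose feeds two base-bone-space RigidBody chains with different physics assets, selected by bIsInFrontEnd (the "_End" asset when true).
# Each chain then uses weighted translation constraints to blend the right-eye bones toward their left counterparts, with different weights per chain.
# The trailing additive bone modifiers carry zero offsets, and the event graph is unwired.
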AnimBlueprint /BRCosmetics/Weapons/FORT_Melee/Pickaxe_VitalInventor/Meshes/VitalInventor_Axe_AnimBP
target_skeleton=&/VitalInventor_Axe_Skeleton
graphs [
 AnimGraph {nodes [
   n1 = RigidBody(ComponentPose=Local To Component(LocalPose=Use cached pose 'BodyPose'())) {
    bOverrideWorldGravity=true
    overridePhysicsAsset=&/VitalInventor_Axe_Physics
    simulationSpace=BaseBoneSpace
    alphaBoolBlend {blendOption=Linear}
    baseBoneRef {boneName=b_AR_Root}
    componentAppliedLinearAccClamp=<2000 2000 2000>
    componentLinearAccScale=<0.3 0.3 0.35>
    componentLinearVelScale=<0.3 0.3 0.4>
    overrideWorldGravity {z=-980}
   }
   n2 = "Constraint - dyn_eye_r_base (TT)"!(ComponentPose=@n1, Alpha=1, ConstraintWeights_0=0.7, ConstraintWeights_1:float=0.3) -> (Pose) {
    alphaBoolBlend {blendOption=Linear}
    boneToModify {boneName=dyn_eye_r_base}
    constraintSetup [{transformType=Translation, targetBone {boneName=dyn_eye_l_base}}, {transformType=Translation, targetBone {boneName=dyn_eye_r_base}}]
    constraintWeights [1,1]
   }
   n3 = "Constraint - dyn_eye_r_01 (TT)"!(ComponentPose=@n2, Alpha=1, ConstraintWeights_0=0.5, ConstraintWeights_1:float=0.5) -> (Pose) {
    alphaBoolBlend {blendOption=Linear}
    boneToModify {boneName=dyn_eye_r_01}
    constraintSetup [{transformType=Translation, targetBone {boneName=dyn_eye_l_01}}, {transformType=Translation, targetBone {boneName=dyn_eye_r_01}}]
    constraintWeights [1,1]
   }
   n4 = "Constraint - dyn_eye_r (TT)"!(ComponentPose=@n3, Alpha=1, ConstraintWeights_0=0.3, ConstraintWeights_1:float=0.7) -> (Pose) {
    alphaBoolBlend {blendOption=Linear}
    boneToModify {boneName=dyn_eye_r}
    constraintSetup [{transformType=Translation, targetBone {boneName=dyn_eye_l}}, {transformType=Translation, targetBone {boneName=dyn_eye_r}}]
    constraintWeights [1,1]
   }
   n5 = "Transform (Modify) Bone - Bone: dyn_eye_r_base"(ComponentPose=@n4, Alpha=1, Translation="0,0,0", Rotation="0,0,0", Scale="1,1,1") {
    rotationMode=BMM_Additive
    rotationSpace=BCS_ParentBoneSpace
    translationMode=BMM_Additive
    translationSpace=BCS_ParentBoneSpace
    alphaBoolBlend {blendOption=Linear}
    boneToModify {boneName=dyn_eye_r_base}
   }
   n6 = "Transform (Modify) Bone - Bone: dyn_eye_r_01"(ComponentPose=@n5, Alpha=1, Translation="0,0,0", Rotation="0,0,0", Scale="1,1,1") {
    rotationMode=BMM_Additive
    rotationSpace=BCS_ParentBoneSpace
    translationMode=BMM_Additive
    translationSpace=BCS_ParentBoneSpace
    alphaBoolBlend {blendOption=Linear}
    boneToModify {boneName=dyn_eye_r_01}
   }
   n7 = "Transform (Modify) Bone - Bone: dyn_eye_r"(ComponentPose=@n6, Alpha=1, Translation="0,0,0", Rotation="0,0,0", Scale="1,1,1") {
    rotationMode=BMM_Additive
    rotationSpace=BCS_ParentBoneSpace
    translationMode=BMM_Additive
    translationSpace=BCS_ParentBoneSpace
    alphaBoolBlend {blendOption=Linear}
    boneToModify {boneName=dyn_eye_r}
   }
   n8 = RigidBody(ComponentPose=Local To Component(LocalPose=Use cached pose 'BodyPose'())) {
    bOverrideWorldGravity=true
    overridePhysicsAsset=&/VitalInventor_Axe_Physics_End
    simulationSpace=BaseBoneSpace
    alphaBoolBlend {blendOption=Linear}
    baseBoneRef {boneName=b_AR_Root}
    componentAppliedLinearAccClamp=<2000 2000 2000>
    componentLinearAccScale=<0.3 0.3 0.35>
    componentLinearVelScale=<0.3 0.3 0.4>
    overrideWorldGravity {z=-980}
   }
   n9 = "Constraint - dyn_eye_r_base (TT)"(ComponentPose=@n8, Alpha=1, ConstraintWeights_0=0.6, ConstraintWeights_1=0.4) {
    alphaBoolBlend {blendOption=Linear}
    boneToModify {boneName=dyn_eye_r_base}
    constraintSetup [{transformType=Translation, targetBone {boneName=dyn_eye_l_base}}, {transformType=Translation, targetBone {boneName=dyn_eye_r_base}}]
    constraintWeights [1,1]
   }
   n10 = "Constraint - dyn_eye_r_01 (TT)"(ComponentPose=@n9, Alpha=1, ConstraintWeights_0=0.3, ConstraintWeights_1=0.7) {
    alphaBoolBlend {blendOption=Linear}
    boneToModify {boneName=dyn_eye_r_01}
    constraintSetup [{transformType=Translation, targetBone {boneName=dyn_eye_l_01}}, {transformType=Translation, targetBone {boneName=dyn_eye_r_01}}]
    constraintWeights [1,1]
   }
   n11 = "Constraint - dyn_eye_r (TT)"(ComponentPose=@n10, Alpha=1, ConstraintWeights_0=0.1, ConstraintWeights_1=0.9) {
    alphaBoolBlend {blendOption=Linear}
    boneToModify {boneName=dyn_eye_r}
    constraintSetup [{transformType=Translation, targetBone {boneName=dyn_eye_l}}, {transformType=Translation, targetBone {boneName=dyn_eye_r}}]
    constraintWeights [1,1]
   }
   n12 = "Transform (Modify) Bone - Bone: dyn_eye_r_base"(ComponentPose=@n11, Alpha=1, Translation="0,0,0", Rotation="0,0,0", Scale="1,1,1") {
    rotationMode=BMM_Additive
    rotationSpace=BCS_ParentBoneSpace
    translationMode=BMM_Additive
    translationSpace=BCS_ParentBoneSpace
    alphaBoolBlend {blendOption=Linear}
    boneToModify {boneName=dyn_eye_r_base}
   }
   n13 = "Transform (Modify) Bone - Bone: dyn_eye_r_01"(ComponentPose=@n12, Alpha=1, Translation="0,0,0", Rotation="0,0,0", Scale="1,1,1") {
    rotationMode=BMM_Additive
    rotationSpace=BCS_ParentBoneSpace
    translationMode=BMM_Additive
    translationSpace=BCS_ParentBoneSpace
    alphaBoolBlend {blendOption=Linear}
    boneToModify {boneName=dyn_eye_r_01}
   }
   n14 = "Transform (Modify) Bone - Bone: dyn_eye_r"(ComponentPose=@n13, Alpha=1, Translation="0,0,0", Rotation="0,0,0", Scale="1,1,1") {
    rotationMode=BMM_Additive
    rotationSpace=BCS_ParentBoneSpace
    translationMode=BMM_Additive
    translationSpace=BCS_ParentBoneSpace
    alphaBoolBlend {blendOption=Linear}
    boneToModify {boneName=dyn_eye_r}
   }
   Output Pose(Result=Blend Poses by bool(BlendPose_0=Component To Local(ComponentPose=@n14), BlendPose_1=Component To Local(ComponentPose=@n7), bActiveValue=Get bIsInFrontEnd()) {blendType=Linear}) {layerGroup=None}
   AnimGraphNode_SubInput_0 = Input Pose()
   Save cached pose 'BodyPose'(Pose=@AnimGraphNode_SubInput_0) {cacheName=BodyPose}
  ]}
 EventGraph {nodes [
   TryGetPawnOwner()
   Event BlueprintUpdateAnimation()
  ]}
]
variables []
